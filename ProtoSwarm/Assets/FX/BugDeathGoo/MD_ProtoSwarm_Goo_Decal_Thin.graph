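# Translucent deferred decal material for the thin goo splat of ProtoSwarm bug deaths.
# Green channels of a decal texture and two half-scale, offset water-splash samples are summed into the splat mask.
# Erode_Alpha takes that saturated mask and inverted decal lifetime opacity scaled by ErodeStrength, so the goo erodes as it expires.
# Color comes from a time-of-day color function, blended toward a DarkColor tint by the splash pattern and multiplied by the mask.
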Material /ProtoSwarm/Assets/FX/BugDeathGoo/MD_ProtoSwarm_Goo_Decal_Thin
graph {
 properties {
  bAutomaticallySetUsageInEditor=false
  blendMode=BLEND_Translucent
  materialDomain=MD_DeferredDecal
  pixelDepthOffsetMode=PDOM_Legacy
  shadingModel=MSM_Unlit
  usedShadingModels=Unlit
 }
 nodes [
  n1 = TextureSampleParameter2D() {parameterName=DecalTexture, texture=/Game/Effects/Fort_Effects/Gameplay/SmokeDustExplosions/Textures/T_Decal_Scorch_Hole}
  n2 = TextureSample(UVs=Multiply(A=TextureCoordinate()) {constB=0.5}) {
   samplerSource=SSM_Clamp_WorldGroupSettings
   samplerType=SAMPLERTYPE_LinearColor
   texture=/Game/Effects/Fort_Effects/Textures/Water/T_waterSplash
  }
  n3 = TextureCoordinate()
  n4 = TextureSample(UVs=Add(A=Multiply(A=@n3) {constB=0.5}, B=Constant2Vector() {r=0.5})) {
   samplerSource=SSM_Clamp_WorldGroupSettings
   samplerType=SAMPLERTYPE_LinearColor
   texture=&T_waterSplash
  }
  n5 = LinearInterpolate(A=@n2.G, B=@n4.G)
  LinearInterpolate(A=@n1.G, B=@n5) {constAlpha=0.8}
  n6 = DecalLifetimeOpacity()
  n7 = Add(A=Add(A=@n2.G, B=@n4.G), B=@n1.G)
  MaterialFunctionCall!(
   Dynamic Param=Multiply(A=OneMinus(None=@n6.Opacity), B=ScalarParameter() {defaultValue=-1, parameterName=ErodeStrength})
   Alpha=Saturate(None=Multiply(A=@n7))
  ) {materialFunction=/Game/Effects/Fort_Effects/Materials/_Functions/Erode_Alpha}
  n8 = VectorParameter() {parameterName=DecalColor_MorningAndNight, defaultValue=<rgba 1 0 1 1>}
  n9 = VectorParameter() {parameterName=DecalColor, defaultValue=<rgba 0.0638 0.0176 0.11 1>}
  n10 = MaterialFunctionCall(Color_Morning=@n8.RGB, Color_Day=@n9.RGB, Color_Evening=@n9.RGB, Color_Night=@n8.RGB) {materialFunction=&Color_Time_of_Day}
  n11 = Multiply(A=@n10.Result, B=VectorParameter() {parameterName=DarkColor, defaultValue=<rgba 0.102 0.102 0.3 0>}.RGB)
  Multiply(A=@n10.Result, B=Multiply(A=LinearInterpolate(A=@n11, B=@n10.Result, Alpha=@n5), B=@n7))
  ScalarParameter() {parameterName=Metallic}
  ScalarParameter() {defaultValue=1, parameterName=Specular}
  ScalarParameter() {defaultValue=1, parameterName=Roughness}
 ]
}
