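# Translucent unlit material for thick ProtoSwarm bug-death goo trails on Niagara ribbons and mesh particles.
# Noise-distorted UVs, steered by dynamic mirror and pan parameters, pick a random flipbook column from a channel-packed goo texture.
# Emissive layers tinted gradients, iridescence and edges with fake sun/moon specular from reflected noise normals.
# SplineThicken plus downward offset shapes the trail; noise erosion, depth fade and near-camera fade drive opacity.
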Material /ProtoSwarm/Assets/FX/BugDeathGoo/M_ProtoSwarm_Goo_Trail_Thick
graph {
 properties {
  bUsedWithNiagaraMeshParticles=true
  bUsedWithNiagaraRibbons=true
  blendMode=BLEND_Translucent
  pixelDepthOffsetMode=PDOM_Legacy
  shadingModel=MSM_Unlit
  translucencyLightingMode=TLM_VolumetricDirectional
  usedShadingModels=Unlit
 }
 nodes [
  n1 = ScalarParameter() {defaultValue=40, parameterName=SplineWidth}
  n2 = LinearInterpolate(A=@n1, B=Multiply(A=@n1) {constB=2}, Alpha=ParticleRandom())
  n3 = Power(
   Base=OneMinus(None=MaterialFunctionCall() {materialFunction=&LinearGradient}.VGradient)
   Exp=Constant() {r=2}
  )
  n4 = Constant3Vector() {constant=<rgba 0 0 -1 0>}
  Add(
   A=MaterialFunctionCall(WidthBase=Divide(A=@n2) {constB=1.5}, WidthTip=@n2) {materialFunction=&SplineThicken}.WorldPositionOffset
   B=Multiply(A=Multiply(A=@n3, B=@n4), B=ScalarParameter() {defaultValue=50, parameterName=WPOoffset})
  )
  Constant() {r=1}
  MaterialFunctionCall() {materialFunction=&DebugTimeSine}
  MaterialFunctionCall() {materialFunction=&DebugTimeSine}
  n5 = TextureSample(UVs=Multiply(A=NamedRerouteUsage(), B=Constant() {r=2})) {samplerType=SAMPLERTYPE_Normal, texture=/Game/Characters/Player/Female/Medium/Bodies/F_MED_Bistro_Astronaut/FX/T_GaussianNoise_Normal}
  Multiply(A=@n5.B)
  n6 = TextureSample(UVs=Multiply(A=NamedRerouteUsage(), B=Constant2Vector() {g=0.5, r=2})) {samplerType=SAMPLERTYPE_LinearGrayscale, texture=/Game/Effects/Fort_Effects/Emotes/UltraLight/Texture/T_DirectionalNoise_Greyscale}
  n7 = Power(Base=MaterialFunctionCall() {materialFunction=&LinearGradient}.VGradient) {constExponent=1}
  Multiply(A=@n6.R, B=Multiply(A=@n7, B=NamedRerouteUsage()) {constA=1})
  n8 = NamedRerouteUsage()
  n9 = TextureSampleParameter2D(UVs=Saturate(None=@n8)) {parameterName=LiquidTrailTex, texture=&/T_ProtoSwarm_Goo_Trail}
  NamedRerouteDeclaration(None=@n9.R) {name=R_Channel_DarkPurpleGoo, nodeColor=<rgba 1 0.008 0 1>}
  NamedRerouteDeclaration(None=@n9.G) {name=G_Channel_OrangeEnergy, nodeColor=<rgba 0.1185 1 0 1>}
  NamedRerouteDeclaration(None=@n9.B) {name=B_Channel_ErodeMask, nodeColor=<rgba 0 0.1069 1 1>}
  n10 = MaterialFunctionCall("Animation  Phase (0-1)"=ParticleRandom(), Number of Rows=Constant() {r=1}, Number of Columns=Constant() {r=4}) {materialFunction=&FlipBook}
  n11 = TextureSample(UVs=Multiply(A=Panner(Coordinate=@n10.UVs) {speedX=0.1, speedY=-0.5}, B=Constant2Vector() {g=1, r=6}) {constB=2}) {texture=/Game/Effects/Fort_Effects/Emotes/UltraLight/Texture/T_SoftNoise_BC7}
  n12 = ComponentMask(None=@n11.RGB) {g=true, r=true}
  n13 = Constant() {r=8}
  n14 = MaterialFunctionCall(UV Channel=Add(A=TextureCoordinate(), B=Divide(A=@n12, B=@n13))) {materialFunction=&LinearGradient}
  n15 = DynamicParameter() {defaultValue=<rgba 1 0 0 1>, paramNames [RandomMirrorU, PanAnim, Erode, Param4]}
  n16 = Multiply(A=@n14.UGradient, B=LinearInterpolate(Alpha=@n15.RandomMirrorU) {constA=-1}) {constB=-1}
  n17 = Add(A=Power(Base=@n14.VGradient, Exp=Constant() {r=0.5}) {constExponent=0.7}, B=LinearInterpolate(Alpha=@n15.PanAnim) {constA=-1, constB=0.05}) {constB=0}
  NamedRerouteDeclaration(None=MaterialFunctionCall(
    "Animation  Phase (0-1)"=ParticleRandom()
    Number of Rows=Constant() {r=1}
    Number of Columns=Constant() {r=4}
    UVs=AppendVector(A=@n16, B=@n17)
   ) {materialFunction=&FlipBook}.UVs) {name=FlipbookUV, nodeColor=<rgba 1 0.9375 0 1>}
  NamedRerouteUsage()
  NamedRerouteUsage()
  NamedRerouteUsage()
  NamedRerouteUsage()
  NamedRerouteUsage()
  NamedRerouteUsage()
  Reroute(None=Multiply(A=@n11.R, B=@n11.R))
  n18 = ScalarParameter() {defaultValue=3, parameterName=IridescentPower}
  n19 = Multiply(A=Add(A=MaterialFunctionCall() {materialFunction=&Color_Time_of_Day}.Result) {constB=0.25}) {constB=0.5}
  n20 = CollectionParameter() {
   collection=&FortniteMaterialParameters
   parameterName=SunAndMoonModelDirectionalVector
  }
  n21 = ComponentMask(None=@n20) {b=true, g=true, r=true}
  n22 = NamedRerouteUsage()
  n23 = NamedRerouteUsage()
  n24 = LinearInterpolate(A=Constant3Vector() {constant=<rgba 0 0 1 0>}, B=@n5.RGB, Alpha=Saturate(None=Subtract(A=@n22, B=@n23)))
  n25 = ReflectionVectorWS(CustomWorldNormal=@n24) {bNormalizeCustomWorldNormal=true}
  n26 = Clamp(None=Power(Base=DotProduct(A=@n21, B=@n25), Exp=ScalarParameter() {defaultValue=12, parameterName=Spec Highlight Power}) {constExponent=20})
  n27 = DotProduct(
   A=ComponentMask(None=Multiply(A=@n20) {constB=-1}) {b=true, g=true, r=true}
   B=ReflectionVectorWS(CustomWorldNormal=@n24) {bNormalizeCustomWorldNormal=true}
  )
  n28 = ScalarParameter() {defaultValue=2, parameterName=Reverse Spec Highlight Power}
  n29 = Multiply(A=@n19, B=Clamp(None=Power(Base=@n27, Exp=@n28) {constExponent=20}))
  n30 = ScalarParameter() {defaultValue=1, parameterName=ReverseHighlightOpacity}
  n31 = Multiply(
   A=Add(A=Multiply(A=@n19, B=@n26), B=Multiply(A=@n29, B=@n30))
   B=VectorParameter() {parameterName=ColorHighlight, defaultValue=<rgba 0.3063 0.1 0.5 1>}.RGB
  )
  n32 = Constant() {r=20}
  n33 = Multiply(
   A=LinearInterpolate(
    A=VectorParameter() {parameterName=ColorA, defaultValue=<rgba 0.092 0.06 0.3 1>}.RGB
    B=VectorParameter() {parameterName=ColorB, defaultValue=<rgba 0.4133 0.12 1 1>}.RGB
    Alpha=@n7
   )
   B=Multiply(A=@n9.G, B=LinearInterpolate(Alpha=@n6.R) {constA=0.2})
  )
  n34 = NamedRerouteUsage()
  n35 = Multiply(A=VectorParameter() {parameterName=ColorEdges, defaultValue=<rgba 0.565 0.13 1 1>}.RGB, B=OneMinus(None=Power(Base=@n34)))
  n36 = Add(
   A=Multiply(A=VectorParameter() {parameterName=ColorIridescent, defaultValue=<rgba 0.2 2 0.29 1>}.RGB, B=Power(Base=@n6.R, Exp=@n18))
   B=Add(A=Multiply(A=@n31, B=@n32), B=Add(A=@n33, B=@n35))
  )
  n37 = ParticleColor()
  Reroute(None=MaterialFunctionCall(Emissive=Multiply(A=@n36, B=@n37.RGB)) {materialFunction=&MF_FX_EmissiveAdjustments}.Result)
  n38 = NamedRerouteUsage()
  n39 = NamedRerouteUsage()
  n40 = Constant2Vector() {g=1, r=2}
  n41 = Multiply(
   A=DynamicParameter() {defaultValue=<rgba 1 0 0 1>, paramNames [RandomMirrorU, PanAnim, Erode, Param4]}.Erode
   B=TextureSample(UVs=Multiply(A=@n39, B=@n40)) {samplerType=SAMPLERTYPE_Grayscale, texture=&LowResBlurredNoise}.R
  )
  n42 = Multiply(
   A=Add(A=Subtract(A=@n38, B=@n41) {constB=0}, B=NamedRerouteUsage())
   B=OneMinus(None=MaterialFunctionCall() {materialFunction=&LinearGradient}.VGradient)
  )
  n43 = Saturate(None=Multiply(A=Saturate(None=@n42), B=ScalarParameter() {defaultValue=3, parameterName=Opacity}))
  n44 = DepthFade(Opacity=@n43, FadeDistance=ScalarParameter() {defaultValue=50, parameterName=DepthFadeDistance}) {fadeDistanceDefault=50}
  n45 = ShadingPathSwitch(Default=@n44, Deferred=@n44, Mobile=@n43) {desc=Migrated from FeatureLevelSwitch}
  n46 = LinearInterpolate(Alpha=MaterialFunctionCall(Distance=ScalarParameter() {defaultValue=50, parameterName=CamFadeDis}) {materialFunction=&Near_Camera_Fade}.Result) {constA=0.1}
  Reroute(None=Multiply(A=ParticleColor().A, B=Multiply(A=@n45, B=@n46)))
  TextureCoordinate()
  TextureSample(UVs=@n8) {samplerSource=SSM_Clamp_WorldGroupSettings, texture=&T_ProtoSwarm_Goo_Trail}
 ]
}
